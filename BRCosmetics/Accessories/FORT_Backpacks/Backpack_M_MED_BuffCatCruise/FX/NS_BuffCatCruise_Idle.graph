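# Niagara idle effect for the BuffCatCruise backpack cosmetic, emitting drops and splashes from a sampled point on the skeletal mesh.
# Drops are cone-launched meshes with gravity, drag, inherited movement and distance-based culling; Splash uses animated sprites mirrored left or right via pivot and negative scale.
# Spawn rates generally follow forward movement, change with jump reactivity and stop during skydiving.
# Time-of-day functions tint and dim both emitters.
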NiagaraSystem /BRCosmetics/Accessories/FORT_Backpacks/Backpack_M_MED_BuffCatCruise/FX/NS_BuffCatCruise_Idle
dependencies {
 usedDataInterfaces [
  &NiagaraDataInterfaceCurve
  &NiagaraDataInterfaceVectorCurve
  &NiagaraDataInterfaceVector2DCurve
 ]
 usedDynamicInputs [
  &Multiply_Float
  /Niagara/DynamicInputs/TypeConversions/Vector2DLength
  &MakeFloatFromVector
  &Abs_Float
  &VectorLength
  &SetBoolByFloatComparison
 ]
 usedModules [
  &SystemState
  &self:SystemSpawnScript.SystemScriptSource.SystemScriptGraph.NiagaraNodeAssignment_3.SetVariables_#1
  &self:SystemSpawnScript.SystemScriptSource.SystemScriptGraph.NiagaraNodeAssignment_0.SetVariables_#2
 ]
}
emitters {
 Drops {
  emitterUpdateScript [
   &EmitterState(Life Cycle Mode=::NewEnumerator1, Loop Duration=0.5, UseLoopDelay=false)
   &SampleSkeletalMesh(
    AdditionalYawPitchRoll=false
    Mesh Sampling Type:ENiagaraSkelSamplingModeFull=::NewEnumerator5
    Skeletal Mesh:NiagaraDataInterfaceSkeletalMesh=$User.SkeletalMesh
    Socket Sampling Mode:ENiagaraSocketSamplingMode=::NewEnumerator5
   )
   &SpawnRate(
    SpawnRate=&MakeCustomFloatFromBool(
     Bool=$User.IsSkydiving
     True Float=0
     False Float=&RemapRange(
      Input Value To Remap=$System.MovingForward
      Input Min=0
      Input Max=1
      Output Min=0
      Output Max=&MakeCustomFloatFromBool(
       Bool=$User.bJumpReactivity
       True Float=&RandomRangeFloat(Minimum=3, Maximum=6)
       False Float=&RandomRangeFloat(Minimum=5, Maximum=7.5)
      )
     )
    )
    Use Spawn Probability=false
   )
  ]
  particleSpawnScript [
   &InitializeParticle(
    Color=<rgba 0.776 0.776 0.776 1>
    Lifetime Max=0.4
    Lifetime Min=0.2
    Lifetime Mode=::NewEnumerator1
    Mass Mode=::NewEnumerator3
    Mesh Scale=&Multiply_VectorByFloat(Vector=<2 1 1>, Float=&RandomRangeFloat(Minimum=0.3, Maximum=0.6))
    Mesh Scale Mode=::NewEnumerator4
    UsePositionOffset=false
    Write Material Random=true
   )
   &CylinderLocation(
    Cylinder Height=1
    Cylinder Origin:NiagaraPosition=&ConvertVectorToPosition(Input Position=$Emitter.SampleSkeletalMesh.SampledPosition, Vector Is In Absolute Large World Space:NiagaraBool=false)
    Cylinder Radius=2
    MaskSpawnGroup=false
    Override Seed=false
    UseOwnerScale=true
   )
   &AddVelocityInCone(
    Cone Angle=25
    Cone Axis=<0 0 0.5>
    Module.Use Velocity Falloff On Cone Axis=false
    Velocity Strength=&RandomRangeFloat(Minimum=150, Maximum=300)
   )
   &SubUVAnimation(End Frame=3, SubUV Animation Mode:ENiagara_SubUVLookupMode=::NewEnumerator2)
  ]
  particleUpdateScript [
   &ParticleState()
   /Game/Effects/Niagara/Emitters/Water/FN_DistanceKillParticles(
    AgeReduiceAmount:NiagaraFloat=&MakeCustomFloatFromBool(Bool=$User.bJumpReactivity, True Float=0.8, False Float=0)
    CenterPosition:Vector3f=$Emitter.SampleSkeletalMesh.SampledPosition
    DistanceFalloffExp:NiagaraFloat=&MakeCustomFloatFromBool(Bool=$User.bJumpReactivity, True Float=1, False Float=0)
   )
   &GravityForce()
   &Drag(
    Drag=&RandomRangeFloat(Minimum=4, Maximum=8, Evaluation Type=::NewEnumerator1)
    Use Linear Drag=true
    Use Rotational Drag=false
   )
   &ScaleColor(
    Scale Alpha=&FloatFromCurve(FloatCurve=DataInterface() {propertyValues {Curve {keys [<at -0.0046 1.0113 RCIM_Cubic>,<at 1.0011 0.002 RCIM_Cubic>]}}})
    Scale RGB=&Multiply_VectorByFloat(
     Vector=&FN_Select_Vector_from_Time_Of_Day(Morning=<1 0.6 0.6>, Day=<1 0.9 0.9>, Evening=<1 0.7 0.7>, Night=<0.75 1 1>, Evening Threshold=19)
     Float=&FNTimeOfDayScalar(Night Scale=0.4, Morning Scale=0.75, Evening Scale=0.4)
    )
    ScaleA=true
    ScaleRGB=true
   )
   &InheritSourceMovement(
    Applied Force Scale=&VectorFromFloat(Value=&FloatFromCurve(FloatCurve=DataInterface() {propertyValues {Curve {keys [{arriveTangent=0.4365, interpMode=RCIM_Cubic, leaveTangent=0.4365, tangentMode=RCTM_User, time=0, value=1}, <at 1 0 RCIM_Cubic>]}}}))
    Applied Position Scale=&VectorFromFloat(Value=&FloatFromCurve(FloatCurve=DataInterface() {propertyValues {Curve {keys [<at 0 1 RCIM_Cubic>,<at 0.9 0 RCIM_Cubic>]}}}))
    ApplyForce=true
    ApplyForceFalloff=false
    ApplyPosition=true
    ApplyPositionFalloff=false
    Module.LimitForceSpeed=false
   )
   &SolveForcesAndVelocity(Clamp Velocity=false, Limit Acceleration=false)
   &ScaleSpriteSize(Uniform Curve Sprite Scale=DataInterface() {propertyValues {Curve {keys [
        <at 0 0 RCIM_Cubic>
        {arriveTangent=1, interpMode=RCIM_Cubic, leaveTangent=1, time=0.1, value=1}
        {arriveTangent=-1.1111, interpMode=RCIM_Cubic, leaveTangent=-1.1111, time=0.7, value=0.7}
        <at 1 0 RCIM_Cubic>
       ]}}})
   &CameraOffset(Camera Offset Amount=1)
   &ScaleMeshSize(Scale Factor=&VectorFromCurve(VectorCurve=DataInterface() {propertyValues {
       XCurve {keys [<at 0 1 RCIM_Cubic>, {arriveTangent=-1, interpMode=RCIM_Cubic, leaveTangent=-1, time=0.5, value=0.875}, <at 1 0 RCIM_Cubic>]}
       YCurve {keys [<at 0 1 RCIM_Cubic>, {arriveTangent=-1, interpMode=RCIM_Cubic, leaveTangent=-1, time=0.5, value=0.9}, <at 1 0 RCIM_Cubic>]}
       ZCurve {keys [<at 0 1 RCIM_Cubic>,<at 1 0.9 RCIM_Cubic>]}
      }}))
  ]
  rendererClasses [&NiagaraMeshRendererProperties]
  renderers [{rendererClass=&NiagaraMeshRendererProperties, rendererIndex=0}]
 }
 Splash {
  emitterUpdateScript [
   &EmitterState(Life Cycle Mode=::NewEnumerator1, Loop Duration=1, UseLoopDelay=false)
   &SampleSkeletalMesh(
    AdditionalYawPitchRoll=false
    Mesh Sampling Type:ENiagaraSkelSamplingModeFull=::NewEnumerator5
    Skeletal Mesh:NiagaraDataInterfaceSkeletalMesh=$User.SkeletalMesh
    Socket Sampling Mode:ENiagaraSocketSamplingMode=::NewEnumerator5
   )
   &SpawnRate(
    SpawnRate=&MakeCustomFloatFromBool(
     Bool=$User.IsSkydiving
     True Float=0
     False Float=&MakeCustomFloatFromBool(
      Bool=$User.bJumpReactivity
      True Float=1
      False Float=&RemapRange(
       Input Value To Remap=$System.MovingForward
       Input Min=0
       Input Max=1
       Output Min=0
       Output Max=&RandomRangeFloat(Minimum=5, Maximum=10)
      )
     )
    )
    Use Spawn Probability=false
   )
  ]
  particleSpawnScript [
   &InitializeParticle(
    Color Mode=::NewEnumerator0
    Lifetime Max=0.5
    Lifetime Min=0.25
    Lifetime Mode=::NewEnumerator1
    Position Offset=<0 0 7.5>
    Sprite Size=<8 4>
    Sprite Size Mode=::NewEnumerator4
    UsePositionOffset=true
    Write Material Random=true
   )
   /SetVariables(Particles.IsLHS:NiagaraBool=&self:ScratchDynamicInput_01()) {isSetParameters=true}
   /SetVariables(Particles.PivotOffset:Vector2f=&MakeCustomVector2DFromBool(Bool=$Particles.IsLHS, True Vector2D=<0.75 1>, False Vector2D=<0.25 1>)) {isSetParameters=true}
   &SubUVAnimation(End Frame=1, SubUV Animation Mode:ENiagara_SubUVLookupMode=::NewEnumerator2)
   &CameraOffset()
   &SystemLocation(MaskSpawnGroup=false, Offset=$Emitter.SampleSkeletalMesh.SampledPosition)
  ]
  particleUpdateScript [
   &ParticleState()
   /SetVariables() {isSetParameters=true}
   /SetVariables(Particles.SpriteAlignment=<0 0 1>) {isSetParameters=true}
   &DynamicMaterialParameters(
    Index 0 Param 1=&Lerp_Float(
     A=0.3
     B=1
     Alpha=&FloatFromCurve(FloatCurve=DataInterface() {propertyValues {Curve {keys [
          {arriveTangent=0.1629, interpMode=RCIM_Cubic, time=0, value=0.975}
          {arriveTangent=0.0278, interpMode=RCIM_Cubic, leaveTangent=0.0278, time=0.2, value=0.025}
          <at 0.9 1 RCIM_Cubic>
         ]}}})
    )
    Module.Param0WriteEnabled=true
    Module.Param1WriteEnabled=true
    Module.Param2WriteEnabled=true
    Module.Param3WriteEnabled=true
   )
   &ScaleSpriteSize(
    Non-Uniform Curve Sprite Scale=DataInterface() {propertyValues {
      XCurve {keys [
        {arriveTangent=2.8681, interpMode=RCIM_Cubic, leaveTangent=2.8681, tangentMode=RCTM_User, time=0, value=1}
        {arriveTangent=1.25, interpMode=RCIM_Cubic, leaveTangent=1.25, time=0.4, value=1.75}
        <at 1 2.25 RCIM_Cubic>
       ]}
      YCurve {keys [
        <at 0 1 RCIM_Cubic>
        {arriveTangent=0.7, interpMode=RCIM_Cubic, leaveTangent=0.7, time=0.1, value=2.425}
        {arriveTangent=-1.5833, interpMode=RCIM_Cubic, leaveTangent=-1.5833, time=0.5, value=1.35}
        <at 1 1 RCIM_Cubic>
       ]}
     }}
    Scale Sprite Size Mode=::NewEnumerator3
   )
   &ScaleSpriteSize(
    Scale Factor=&MakeVector2D(X=&MakeCustomFloatFromBool(Bool=$Particles.IsLHS, True Float=-1, False Float=1), Y=1)
    Scale Sprite Size Mode=::NewEnumerator2
   ) {module=ScaleSpriteSize001}
   &ScaleColor(
    Scale RGB=&Multiply_VectorByFloat(
     Vector=&FN_Select_Vector_from_Time_Of_Day(Morning=<1 0.6 0.6>, Day=<1 0.9 0.9>, Evening=<1 0.7 0.7>, Night=<0.75 1 1>, Evening Threshold=19)
     Float=&FNTimeOfDayScalar(Night Scale=0.2, Morning Scale=0.65, Day Scale=0.8, Evening Scale=0.4)
    )
    ScaleA=true
    ScaleRGB=true
   )
  ]
  rendererClasses [&NiagaraSpriteRendererProperties]
  renderers [{rendererClass=&NiagaraSpriteRendererProperties}]
 }
}
summary {
 systemName=NS_BuffCatCruise_Idle
 emitters [
  {emitterName=Drops, rendererClasses [&NiagaraMeshRendererProperties]}
  {emitterName=Splash, rendererClasses [&NiagaraSpriteRendererProperties]}
 ]
 userVariables [
  User.bJumpReactivity {type=NiagaraBool}
  User.GlobalOffset {type=Vector3f, description="", defaultValue {struct=Vector3f, value=<0 0 0>}}
  User.IsSkydiving {type=NiagaraBool}
  User.RibbonIDOffset {type=NiagaraInt32}
  User.SkeletalMesh {
   type=NiagaraDataInterfaceSkeletalMesh
   defaultValue {
    struct=NiagaraExt_VariableValue_DataInterface
    value {dataInterface=&self:NiagaraDataInterfaceSkeletalMesh_0}
   }
  }
 ]
}
