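# Physics asset for the BuffCatCruise backpack cosmetic, driving secondary motion on a drink cup with straw, leaves, fruit and milk.
# Every body is a single capsule, with the cup's the widest.
# The cup sways within tight limits on base_cup, which rotates freely against the pack, and carries a three-link straw, two two-link leaf chains, a strawberry and a lime.
# Milk hangs from the cup through a free-rotating base_milk and a tightly limited dyn_milk.
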PhysicsAsset /BRCosmetics/Accessories/FORT_Backpacks/Backpack_M_MED_BuffCatCruise/Meshes/M_MED_BuffCatCruise_Pack_Physics
bodies {
 base_cup [{shapeType=Capsule, length=1, radius=1, shapeName=base_cup_capsule, center=<1 0 0>, rotation=<rot -90 0 0>}]
 base_milk [{shapeType=Capsule, length=0.505, radius=0.505, shapeName=base_milk_capsule, rotation=<rot 0 -90 0>}]
 dyn_cup [{shapeType=Capsule, length=3, radius=13, shapeName=dyn_cup_capsule, center=<1 0 0>, rotation=<rot -90 0 0>}]
 dyn_leaf_a_01 [{shapeType=Capsule, length=1, radius=1, shapeName=dyn_leaf_a_01_capsule, center=<1 0 0>, rotation=<rot -90 0 0>}]
 dyn_leaf_a_02 [{shapeType=Capsule, length=1, radius=1, shapeName=dyn_leaf_a_02_capsule, center=<1 0 0>, rotation=<rot -90 0 0>}]
 dyn_leaf_b_01 [{shapeType=Capsule, length=1, radius=1, shapeName=dyn_leaf_b_01_capsule, center=<1 0 0>, rotation=<rot -90 0 0>}]
 dyn_leaf_b_02 [{shapeType=Capsule, length=1, radius=1, shapeName=dyn_leaf_b_02_capsule, center=<1 0 0>, rotation=<rot -90 0 0>}]
 dyn_lime [{shapeType=Capsule, length=1, radius=1, shapeName=dyn_lime_capsule, center=<1 0 0.2985>, rotation=<rot -90 0 0>}]
 dyn_milk [{shapeType=Capsule, length=0.1, radius=0.1, shapeName=dyn_milk_capsule, center=<0.1 0 0>, rotation=<rot -90 0 0>}]
 dyn_straw_1 [{shapeType=Capsule, length=7.7869, radius=0.58, shapeName=dyn_straw_1_capsule, center=<5.905 0 0>, rotation=<rot -90 0 0>}]
 dyn_straw_2 [{shapeType=Capsule, length=0.5, radius=0.58, shapeName=dyn_straw_2_capsule, center=<0.5 0 0>, rotation=<rot -90 0 0>}]
 dyn_straw_3 [{shapeType=Capsule, length=0.5, radius=0.58, shapeName=dyn_straw_3_capsule, center=<1 0 0>, rotation=<rot -90 0 0>}]
 dyn_strawberry [{shapeType=Capsule, length=1, radius=1, shapeName=dyn_strawberry_capsule, center=<1 0 0>, rotation=<rot -90 0 0>}]
 pack [{shapeType=Capsule, length=1, radius=1, shapeName=pack_capsule, rotation=<rot -90 0 0>}]
}
constraints [
 {
  bone1Name=base_cup
  bone2Name=pack
  swing1LimitDegrees=45
  swing1Motion=Free
  swing2LimitDegrees=45
  swing2Motion=Free
  twistLimitDegrees=45
  twistMotion=Free
 }
 {bone1Name=dyn_leaf_a_01, bone2Name=dyn_cup, swing1LimitDegrees=15, swing2LimitDegrees=2.5, twistLimitDegrees=0}
 {bone1Name=dyn_leaf_a_02, bone2Name=dyn_leaf_a_01, swing1LimitDegrees=15, swing2LimitDegrees=1.5, twistLimitDegrees=0}
 {bone1Name=dyn_leaf_b_01, bone2Name=dyn_cup, swing1LimitDegrees=15, swing2LimitDegrees=2.5, twistLimitDegrees=0}
 {bone1Name=dyn_leaf_b_02, bone2Name=dyn_leaf_b_01, swing1LimitDegrees=15, swing2LimitDegrees=1.5, twistLimitDegrees=0}
 {bone1Name=dyn_milk, bone2Name=base_milk, swing1LimitDegrees=2.1, swing2LimitDegrees=2.1, twistLimitDegrees=0}
 {bone1Name=dyn_straw_1, bone2Name=dyn_cup, swing1LimitDegrees=15, swing2LimitDegrees=15, twistLimitDegrees=5}
 {bone1Name=dyn_straw_2, bone2Name=dyn_straw_1, swing1LimitDegrees=5, swing2LimitDegrees=7.5, twistLimitDegrees=0}
 {bone1Name=dyn_straw_3, bone2Name=dyn_straw_2, swing1LimitDegrees=5, swing2LimitDegrees=7.5, twistLimitDegrees=0}
 {
  bone1Name=base_milk
  bone2Name=dyn_cup
  swing1LimitDegrees=45
  swing1Motion=Free
  swing2LimitDegrees=45
  swing2Motion=Free
  twistLimitDegrees=45
  twistMotion=Free
 }
 {bone1Name=dyn_cup, bone2Name=base_cup, swing1LimitDegrees=2.1, swing2LimitDegrees=2.1, twistLimitDegrees=2}
 {bone1Name=dyn_strawberry, bone2Name=dyn_cup, swing1LimitDegrees=5, swing2LimitDegrees=10, twistLimitDegrees=0}
 {bone1Name=dyn_lime, bone2Name=dyn_cup, swing1LimitDegrees=2, swing2LimitDegrees=12, twistLimitDegrees=0}
]
