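# Anim Blueprint for the BuffCatCruise backpack cosmetic, adding secondary motion to the pack on the player.
# The base_cup bone turns by the character's spine_05 yaw, or a fixed yaw when DBNO, then feeds RigidBody physics.
# Physics uses overridden gravity, front-end or in-game sim-space settings, and a reduced alpha while parachuting.
# An interpolated additive offset then nudges dyn_strawberry during jumps outside skydiving.
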AnimBlueprint /BRCosmetics/Accessories/FORT_Backpacks/Backpack_M_MED_BuffCatCruise/Meshes/M_MED_BuffCatCruise_Pack_AnimBP
target_skeleton=&/M_MED_BuffCatCruise_Pack_Skeleton
graphs [
 AnimGraph {nodes [
   n1 = "Transform (Modify) Bone - Bone: base_cup"(
    ComponentPose=Local To Component(LocalPose=Use cached pose 'InputCache'())
    Alpha=1
    Translation="0,0,0"
    Rotation=Get GimbalRotation() -> (GimbalRotation:Rotator)
    Scale="1,1,1"
   ) {rotationMode=BMM_Additive, rotationSpace=BCS_BoneSpace, alphaBoolBlend {blendOption=Linear}, boneToModify {boneName=base_cup}}
   n2 = Component To Local(ComponentPose=@n1)
   n3 = "Transform (Modify) Bone - Bone: base_cup"(ComponentPose=Local To Component(LocalPose=Use cached pose 'InputCache'()), Alpha=1, Translation="0,0,0", Rotation="0,90,0", Scale="1,1,1") {
    rotationMode=BMM_Additive
    rotationSpace=BCS_BoneSpace
    alphaBoolBlend {blendOption=Linear}
    boneToModify {boneName=base_cup}
    rotation {yaw=90}
   }
   n4 = Component To Local(ComponentPose=@n3)
   n5 = Get bIsDBNO()
   n6 = RigidBody!(
    ComponentPose=Local To Component(LocalPose=Blend Poses by bool(BlendPose_0=@n4, BlendPose_1=@n2, bActiveValue=@n5) {blendType=QuinticInOut})
    Alpha=Get rigidbody_alpha() -> (rigidbody_alpha:double)
    OverrideWorldGravity:Vector=Get GravityOverride()
    ExternalForce
    ComponentLinearAccScale:Vector="0.6,0.6,0.8"
    ComponentLinearVelScale:Vector="0.6,0.6,0.8"
    SimSpaceSettings:Sim Space Settings Structure=Get Sim Space Settings() -> (Sim Space Settings:Sim Space Settings Structure)
   ) -> (Pose) {
    bClampLinearTranslationLimitToRefPose=true
    bOverrideWorldGravity=true
    overridePhysicsAsset=&/M_MED_BuffCatCruise_Pack_Physics
    simulationSpace=BaseBoneSpace
    alphaBoolBlend {blendOption=Linear}
    alphaScaleBiasClamp {bInterpResult=true}
    baseBoneRef {boneName=root}
    componentAppliedLinearAccClamp=<3000 3000 3000>
    componentLinearAccScale=<0.4 0.3 0.3>
    componentLinearVelScale=<0.4 0.3 0.3>
    simSpaceSettings {worldAlpha=1}
   }
   n7 = "Transform (Modify) Bone - Bone: dyn_strawberry"(ComponentPose=@n6, Alpha=Get is jumping float() -> (is jumping float:double), Translation="2,0,-2", Rotation="-15,0,0", Scale="1,1,1") {
    rotationMode=BMM_Additive
    rotationSpace=BCS_BoneSpace
    translationMode=BMM_Additive
    translationSpace=BCS_BoneSpace
    alphaBoolBlend {blendOption=Linear}
    alphaScaleBiasClamp {bInterpResult=true, interpSpeedDecreasing=2.5, interpSpeedIncreasing=2.5}
    boneToModify {boneName=dyn_strawberry}
    rotation {pitch=-15}
    translation {x=2, z=-2}
   }
   Output Pose(Result=Component To Local(ComponentPose=@n7)) {layerGroup=None}
   Save cached pose 'InputCache'(Pose=Input Pose()) {cacheName=InputCache}
  ]}
 EventGraph {nodes [
   Event BlueprintUpdateAnimation()
   > n1 = Gravity Override State Function DEPRECATED(CharacterPartAnimInstance=Self-Reference(), GameTimeInSeconds=GetGameTimeInSeconds(), GravityOverrideParams=Get Gravity Override Params())
   > Set GravityOverride(GravityOverride=@n1)
   n2 = GetOwnerSkeletalMeshComponent()
   n3 = BreakRotator(InRot=BreakTransform(InTransform=GetDeltaTransformFromRefPose(self=@n2, BoneName=spine_05, BaseName=root)).Rotation)
   > Set GimbalRotation(GimbalRotation:Rotator=MakeRotator(Yaw=float * float(A=@n3.Yaw, B=1) {is=title K2Node_PromotableOperator})) -> (Output_Get:Rotator)
   n4 = NOT Boolean(A=Get bIsSkyDiving())
   n5 = Get bIsJumping()
   > Set is jumping float(is jumping float:double="To Float (Boolean)"(InBool=AND Boolean(A=@n4, B=@n5))) -> (Output_Get:double)
   > n6 = Branch(Condition=Get bIsParachuteOpened())
   > n7 = Set rigidbody_alpha(rigidbody_alpha:double=0.25) -> (Output_Get:double)
   n8 = Branch(execute=[@n9.then, @n7.then], Condition=Get bIsInFrontEnd())
   > Set Sim Space Settings(Sim Space Settings:Sim Space Settings Structure=Get Sim Space Settings_FrontEnd() -> (Sim Space Settings_FrontEnd:Sim Space Settings Structure)) -> (Output_Get:Sim Space Settings Structure)
   Set Sim Space Settings(execute=@n8.else, Sim Space Settings=Get Sim Space Settings_InZone() -> (Sim Space Settings_InZone:Sim Space Settings Structure))
   n9 = Set rigidbody_alpha(execute=@n6.else, rigidbody_alpha=1)
   Event BlueprintInitializeAnimation()
   > n10 = Cast To FortPlayerPawn(Object=TryGetFortPawn())
   > Set Skeletal Mesh(Skeletal Mesh:Skeletal Mesh Component Object Reference=GetSkeletalMeshForPartType(self=@n10)) -> (Output_Get:Skeletal Mesh Component Object Reference)
   TryGetPawnOwner()
  ]}
]
variables [
 GimbalRotation {type=Rotator}
 Gravity Override Params {type=GravityOverrideParamsStruct}
 GravityOverride {type=Vector}
 is jumping float {type=number}
 rigidbody_alpha {type=number}
 Sim Space Settings {
  type=SimSpaceSettings
  default {
   dampingAlpha=1
   maxAngularAcceleration=10000
   maxAngularVelocity=10000
   maxLinearAcceleration=10000
   maxLinearVelocity=10000
   velocityScaleZ=1
   worldAlpha=0
   externalLinearDragV=<0 0 0>
  }
 }
 Sim Space Settings_FrontEnd {
  type=SimSpaceSettings
  default {
   dampingAlpha=1
   maxAngularAcceleration=10000
   maxAngularVelocity=10000
   maxLinearAcceleration=10000
   maxLinearVelocity=10000
   velocityScaleZ=1
   worldAlpha=0.25
   externalLinearDragV=<0 0 0>
  }
 }
 Sim Space Settings_InZone {
  type=SimSpaceSettings
  default {
   dampingAlpha=1
   maxAngularAcceleration=10000
   maxAngularVelocity=10000
   maxLinearAcceleration=10000
   maxLinearVelocity=10000
   velocityScaleZ=1
   worldAlpha=0.125
   externalLinearDragV=<0 0 0>
  }
 }
 Skeletal Mesh {type=SkeletalMeshComponent}
]
